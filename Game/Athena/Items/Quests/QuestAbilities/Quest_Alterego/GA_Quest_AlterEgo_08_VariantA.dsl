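; Alter Ego quest gameplay ability (08, Variant A) that swaps the player's cosmetics once they leave a hideout after completing a quest objective.
; Activation defers to the parent; the quest-completion handler validates the objective and sets the cosmetic rarity level.
; It then waits once for the HidingProp StopHiding gameplay event and delays by a configured time.
; Finally it plays FX, applies the new cosmetics, unbinds from the objective-state dispatcher and ends.
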(event Ability|EventFailedToActivatePassiveAbility
  (params (ActorInfo :type Gameplay%20Ability%20Actor%20Info%20Structure)))

(event Ability|EventActivateAbility
  (|CallParentFunction))

(event EventQuestCompleted
  (params (Controller :type Fort%20Player%20Controller%20Object%20Reference) (QuestDef :type Fort%20Quest%20Item%20Definition%20Object%20Reference) (BackendName :type Name :default "None") (CompletionCount :type Integer) (ObjectiveCompleted :type Boolean) (QuestCompleted :type Boolean))
  (bind _objectivevalid (CallFunction|CheckQuestObjective QuestDef ObjectiveCompleted))
  (if _objectivevalid
    (CallFunction|SetCosmeticRarityLevel)
    (Ability|Tasks|WaitGameplayEvent "(TagName=\"Event.Athena.HidingProp.StopHiding\")" 0 true
      (:EventReceived
        (Utilities|FlowControl|Delay (Variables|GAVars|Setup|GetDelayAfterExitHideout))
        (CallFunction|PlayFx)
        (CallFunction|SetNewCosmetics)
        (EventDispatchers|UnbindEventfromOnQuestObjectiveStateChanged (Fortnite|GetFortPlayerControllerfromActor (Variables|GAVars|Data|GetOwningPlayerPawn)) (delegate EventQuestCompleted))
        (Ability|EndAbility)))))
