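; Quest gameplay ability for the Unfused "carry teammate in storm" quest, granting progress when the player carries a teammate in the storm.
; On activation it takes CarriedPawn from the player's attached actors and ends unless that is a teammate.
; A storm tag check either grants progress immediately or starts a looping timer that rechecks carry and storm tags.
; Stopping replaces that loop with a one-shot timer and ends the ability.
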(event Custom|StopChecking
  ;[1 1 1] Stop checking for tags loop
  (Utilities|Time|SetTimerbyEvent (EventDispatchers|CreateEvent :!event Custom|CheckTags) 1.0)
  (Ability|EndAbility))

(event Ability|EventActivateAbilityFromEvent
  (params (EventData :type Gameplay%20Event%20Data%20Structure%20%28by%20ref%29))
  (bind _carriedpawn (Variables|Default|GetCarriedPawn))
  ;[1 1 1] Check for attatched actors and if succesful make them the carried pawn.
  (|CallParentFunction EventData)
  (for _array_element (Actor|GetAttachedActors (Variables|GAVars|Data|GetOwningPlayerPawn) true false)
    (Utilities|IsValid _array_element
      (:Is%20Valid
        (bind _asfort_player_pawn_athena (Utilities|Casting|CastToFortPlayerPawnAthena _array_element))
        (Variables|Default|SetCarriedPawn _asfort_player_pawn_athena))))
  ;[1 1 1] Check for team affiliation, if teammate proceed if enemy end ability
  (Utilities|IsValid _carriedpawn
    (:Is%20Valid
      (if (Team|OnSameTeam _carriedpawn (Variables|GAVars|Data|GetOwningPlayerPawn))
        (if (GameplayTags|TagContainer|CompareTagAssetInterfacetoOtherTags (Variables|Default|GetT_InStorm) (Variables|GAVars|Data|GetOwningPlayerPawn))
          ;[1 1 1] Check if the player is in the storm initially, if not check for player entering the storm every 5 secs.
          (Utilities|Time|SetTimerbyEvent (EventDispatchers|CreateEvent :!event Custom|CheckTags) 2.5 true)
          (else
            (CallFunction|GrantQuestProgress)))
        (else
          (Ability|EndAbility))))
    (:Is%20Not%20Valid
      (Ability|EndAbility))))

(event Custom|CheckTags
  ;[1 1 1] Check for tags being updated
  (if (GameplayTags|TagContainer|CompareTagAssetInterfacetoOtherTags (Variables|Default|GetT_CarryingPlayer) (Variables|GAVars|Data|GetOwningPlayerPawn))
    (if (not (GameplayTags|TagContainer|CompareTagAssetInterfacetoOtherTags (Variables|Default|GetT_InStorm) (Variables|GAVars|Data|GetOwningPlayerPawn)))
      (CallFunction|GrantQuestProgress))
    (else
      (CallFunction|StopChecking))))

(event Custom|GrantQuestProgress
  ;[1 1 1] Grant quest progress and stop checking for tags
  (Quest|SendComplexCustomStatEvent (Variables|GAVars|Data|GetOwningPlayerQuestManager) 0 "(GameplayTags=((TagName=\"Quest.Metadata.Carry.Teammate.InStorm\")))" "(GameplayTags=)" 1)
  (CallFunction|StopChecking)
  (Ability|EndAbility))
